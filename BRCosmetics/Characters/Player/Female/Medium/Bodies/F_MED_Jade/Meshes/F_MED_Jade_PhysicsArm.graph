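# Physics asset for the female medium Jade character body, covering dynamic jacket fronts, a strap, a zipper and a right-arm zipper.
# Capsule bodies on the pelvis, upper spine, legs and right upper arm form the skeletal base, linked by mostly free constraints.
# Tightly limited constraints hang the jacket flaps from spine_03, and the three-bone strap and two-bone zipper chains from spine_05.
# The arm zipper's master bone is angularly locked to upperarm_r, leaving its child slight swing.
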PhysicsAsset /BRCosmetics/Characters/Player/Female/Medium/Bodies/F_MED_Jade/Meshes/F_MED_Jade_PhysicsArm
bodies {
 calf_l [{shapeType=Capsule, length=25, radius=3, shapeName=calf_r_capsule, center=<-31.1047 -2.4698 -0.3277>, rotation=<rot 90 0 0>}]
 calf_r [{shapeType=Capsule, length=25, radius=3, shapeName=calf_r_capsule, center=<31.1047 2.4698 0.3276>, rotation=<rot 90 0 0>}]
 dyn_armzipper [{
   shapeType=Capsule
   length=1.1695
   radius=1.7403
   shapeName=dyn_armzipper_capsule
   center=<2.6111 -0.1397 0.5524>
   rotation=<rot -2.0801 -84.1594 -89.9998>
  }]
 dyn_armzipper_MSTR [{shapeType=Capsule, length=0.505, radius=0.505, shapeName=dyn_armzipper_MSTR_capsule, rotation=<rot 0 -90 0>}]
 dyn_jacketfront_l [{shapeType=Capsule, length=2, radius=2, shapeName=dyn_jacketfront_l_capsule, center=<0 1.491 0>}]
 dyn_jacketfront_r [{shapeType=Capsule, length=2, radius=2, shapeName=dyn_jacketfront_r_capsule, center=<0 1.491 0>}]
 dyn_strap_00_MSTR [{
   shapeType=Capsule
   length=1.3518
   radius=1.1779
   shapeName=dyn_strap_00_MSTR_capsule
   center=<0.6021 -0.0877 -0.0968>
   rotation=<rot 0 -110.0987 -89.9246>
  }]
 dyn_strap_01 [{shapeType=Capsule, length=1.4698, radius=1.6385, shapeName=dyn_strap_01_capsule, center=<3.4187 -0.4023 -0.2323>, rotation=<rot 0 0 60.0003>}]
 dyn_strap_02 [{
   shapeType=Capsule
   length=1.7221
   radius=1.6873
   shapeName=dyn_strap_02_capsule
   center=<2.072 -0.2552 -0.16>
   rotation=<rot -59.257 82.8981 -173.8879>
  }]
 dyn_zipper [{shapeType=Capsule, length=1, radius=1, shapeName=dyn_zipper_capsule, center=<3.2923 0 0>}]
 dyn_zipper_MSTR [{shapeType=Capsule, length=1, radius=1, shapeName=dyn_zipper_MSTR_capsule, center=<0.6275 1.0362 0>}]
 foot_l [{shapeType=Capsule, length=0.505, radius=0.505, shapeName=foot_l_capsule, center=<-1.3229 3.3853 -0.795>, rotation=<rot 0 172.5813 -89.1242>}]
 foot_r [{shapeType=Capsule, length=0.505, radius=0.505, shapeName=foot_r_capsule, center=<1.3227 -3.3854 0.795>, rotation=<rot 0 172.5813 -89.1242>}]
 pelvis [{shapeType=Capsule, length=0.505, radius=0.505, shapeName=pelvis_capsule}]
 spine_03 [{shapeType=Capsule, length=4, radius=8, shapeName=spine_03_capsule, center=<0 0.955 0>}]
 spine_04 [{shapeType=Capsule, length=5, radius=5, shapeName=spine_04_capsule}]
 spine_05 [{shapeType=Capsule, length=7, radius=7.75, shapeName=spine_05_capsule, center=<-0.8542 1.8481 0>}]
 thigh_l [{shapeType=Capsule, length=0.505, radius=0.505, shapeName=thigh_l_capsule, center=<-30.1449 1.3572 0.6958>, rotation=<rot 0 84.5864 -88.0758>}]
 thigh_r [{shapeType=Capsule, length=0.505, radius=0.505, shapeName=thigh_r_capsule, center=<30.1544 -1.3625 -0.6687>, rotation=<rot 0 84.2417 -88.893>}]
 upperarm_r [{
   shapeType=Capsule
   length=7.567
   radius=4.4669
   shapeName=upperarm_r_capsule
   center=<-12.0657 0.7835 1.1773>
   rotation=<rot 19.7551 -89.9999 -97.9073>
  }]
}
constraints [
 {bone1Name=spine_05, bone2Name=spine_04, swing1LimitDegrees=45, swing2LimitDegrees=45, twistLimitDegrees=45, twistMotion=Free}
 {bone1Name=spine_04, bone2Name=spine_03, swing1LimitDegrees=45, swing2LimitDegrees=45, twistLimitDegrees=45, twistMotion=Free}
 {
  bone1Name=thigh_l
  bone2Name=pelvis
  swing1LimitDegrees=45
  swing1Motion=Free
  swing2LimitDegrees=45
  swing2Motion=Free
  twistLimitDegrees=45
  twistMotion=Free
 }
 {
  bone1Name=calf_l
  bone2Name=thigh_l
  swing1LimitDegrees=45
  swing1Motion=Free
  swing2LimitDegrees=45
  swing2Motion=Free
  twistLimitDegrees=45
  twistMotion=Free
 }
 {
  bone1Name=thigh_r
  bone2Name=pelvis
  swing1LimitDegrees=45
  swing1Motion=Free
  swing2LimitDegrees=45
  swing2Motion=Free
  twistLimitDegrees=45
  twistMotion=Free
 }
 {
  bone1Name=calf_r
  bone2Name=thigh_r
  swing1LimitDegrees=45
  swing1Motion=Free
  swing2LimitDegrees=45
  swing2Motion=Free
  twistLimitDegrees=45
  twistMotion=Free
 }
 {
  bone1Name=foot_r
  bone2Name=calf_r
  swing1LimitDegrees=45
  swing1Motion=Free
  swing2LimitDegrees=45
  swing2Motion=Free
  twistLimitDegrees=45
  twistMotion=Free
 }
 {
  bone1Name=foot_l
  bone2Name=calf_l
  swing1LimitDegrees=45
  swing1Motion=Free
  swing2LimitDegrees=45
  swing2Motion=Free
  twistLimitDegrees=45
  twistMotion=Free
 }
 {bone1Name=dyn_jacketfront_l, bone2Name=spine_03, swing1LimitDegrees=20, swing2LimitDegrees=20, twistLimitDegrees=20}
 {bone1Name=dyn_jacketfront_r, bone2Name=spine_03, swing1LimitDegrees=20, swing2LimitDegrees=20, twistLimitDegrees=20}
 {
  bone1Name=spine_03
  bone2Name=pelvis
  swing1LimitDegrees=45
  swing1Motion=Free
  swing2LimitDegrees=45
  swing2Motion=Free
  twistLimitDegrees=45
  twistMotion=Free
 }
 {bone1Name=dyn_strap_00_MSTR, bone2Name=spine_05, swing1LimitDegrees=20, swing2LimitDegrees=20, twistLimitDegrees=10}
 {bone1Name=dyn_strap_01, bone2Name=dyn_strap_00_MSTR, swing1LimitDegrees=20, swing2LimitDegrees=15, twistLimitDegrees=5}
 {bone1Name=dyn_strap_02, bone2Name=dyn_strap_01, swing1LimitDegrees=20, swing2LimitDegrees=15, twistLimitDegrees=5}
 {bone1Name=dyn_zipper, bone2Name=dyn_zipper_MSTR, swing1LimitDegrees=10, swing2LimitDegrees=10, twistLimitDegrees=0}
 {bone1Name=dyn_zipper_MSTR, bone2Name=spine_05, swing1LimitDegrees=20, swing2LimitDegrees=20, twistLimitDegrees=5}
 {
  bone1Name=dyn_armzipper_MSTR
  bone2Name=upperarm_r
  swing1LimitDegrees=20
  swing1Motion=Locked
  swing2LimitDegrees=20
  swing2Motion=Locked
  twistLimitDegrees=5
  twistMotion=Locked
 }
 {bone1Name=dyn_armzipper, bone2Name=dyn_armzipper_MSTR, swing1LimitDegrees=2, swing2LimitDegrees=10, twistLimitDegrees=0}
 {
  bone1Name=upperarm_r
  bone2Name=spine_05
  swing1LimitDegrees=45
  swing1Motion=Free
  swing2LimitDegrees=45
  swing2Motion=Free
  twistLimitDegrees=45
  twistMotion=Free
 }
]
