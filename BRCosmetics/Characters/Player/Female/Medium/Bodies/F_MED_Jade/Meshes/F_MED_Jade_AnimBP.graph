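# Post-process style AnimBlueprint for the F_MED_Jade female body cosmetic, copying the attached parent mesh's pose and adding deformation and secondary motion.
# Twist and sculpt Control Rigs run first, then weighted multi-bone constraints anchor the zipper and arm zipper.
# Two RigidBody simulations run relative to spine_05 and upperarm_r, followed by a strap constraint.
# AnimDynamics on zipper, strap and arm zipper bones finish the pose; the event graph has no logic.
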AnimBlueprint /BRCosmetics/Characters/Player/Female/Medium/Bodies/F_MED_Jade/Meshes/F_MED_Jade_AnimBP
target_skeleton=&/F_MED_Jade_Skeleton
graphs [
 AnimGraph {nodes [
   n1 = Control Rig(Source=Copy Pose From Mesh() {bUseAttachedParent=true}) {
    alphaBoolBlend {blendOption=Linear}
    controlRigAssetReference {blueprintRigClass=&/F_MED_Jade_Deform_CtrlRig_twisties.F_MED_Jade_Deform_CtrlRig_twisties_C}
   }
   n2 = Control Rig!(
    Source=@n1
    Alpha
    KneeRot_fr_l_114deg:Vector
    KneeTrans_fr_l_114deg:Vector="0, 0, 0"
    KneeScale_fr_l_114deg:Vector="0, 0, 0"
    KneeScale_fr_l_130deg:Vector="0, 0, 0"
    KneeRot_fr_l_130deg:Vector="0, 0, 0"
    KneeTrans_fr_l_130deg:Vector="0, 0, 0"
   ) -> (Pose) {
    alphaBoolBlend {blendOption=Linear}
    controlRigAssetReference {blueprintRigClass=&/F_MED_Jade_CtrlRig_sculpts.F_MED_Jade_CtrlRig_sculpts_C}
   }
   n3 = "Constraint - dyn_armzipper_MSTR (PPPPPPPPPP)"!(
    ComponentPose=Local To Component(LocalPose=@n2)
    Alpha=1
    ConstraintWeights_0=0.0012
    ConstraintWeights_1:float=0.0016
    ConstraintWeights_2:float=0.0123
    ConstraintWeights_3:float=0
    ConstraintWeights_4:float=0.0436
    ConstraintWeights_5:float=0.183
    ConstraintWeights_6:float=0.7286
    ConstraintWeights_7:float=0
    ConstraintWeights_8:float=0.0192
    ConstraintWeights_9:float=0.0104
   ) -> (Pose) {
    alphaBoolBlend {blendOption=Linear}
    boneToModify {boneName=dyn_armzipper_MSTR}
    constraintSetup [
     {targetBone {boneName=spine_04}}
     {targetBone {boneName=spine_05}}
     {targetBone {boneName=clavicle_r}}
     {targetBone {boneName=lowerarm_r}}
     {targetBone {boneName=upperarm_r}}
     {targetBone {boneName=upperarm_twist_02_r}}
     {targetBone {boneName=upperarm_twist_01_r}}
     {targetBone {boneName=deform_elbow_fr_r}}
     {targetBone {boneName=deform_deltoid_fr_r}}
     {targetBone {boneName=deform_pectoral_fr_r}}
    ]
    constraintWeights [1,1,1,1,1,1,1,1,1,1]
   }
   n4 = "Constraint - dyn_zipper (PPPPPPPP)"!(
    ComponentPose=@n3
    Alpha=1
    ConstraintWeights_0=0.0745
    ConstraintWeights_1:float=0
    ConstraintWeights_2:float=0.848
    ConstraintWeights_3:float=0.0001
    ConstraintWeights_4:float=0.0001
    ConstraintWeights_5:float=0.0002
    ConstraintWeights_6:float=0.0002
    ConstraintWeights_7:float=0.0769
   ) -> (Pose) {
    alphaBoolBlend {blendOption=Linear}
    boneToModify {boneName=dyn_zipper}
    constraintSetup [
     {targetBone {boneName=spine_04}}
     {targetBone {boneName=spine_03}}
     {targetBone {boneName=spine_05}}
     {targetBone {boneName=clavicle_l}}
     {targetBone {boneName=clavicle_r}}
     {targetBone {boneName=deform_pectoral_fr_l}}
     {targetBone {boneName=deform_pectoral_fr_r}}
     {targetBone {boneName=dyn_zipper}}
    ]
    constraintWeights [1,1,1,1,1,1,1,1]
   }
   n5 = RigidBody(ComponentPose=@n4) {
    bClampLinearTranslationLimitToRefPose=true
    bOverrideWorldGravity=true
    overridePhysicsAsset=&/F_MED_Jade_Physics
    simulationSpace=BaseBoneSpace
    alphaBoolBlend {blendOption=Linear}
    baseBoneRef {boneName=spine_05}
    componentAppliedLinearAccClamp=<3000 3000 3000>
    componentLinearAccScale=<0.4 0.3 0.3>
    componentLinearVelScale=<0.4 0.3 0.3>
    overrideWorldGravity {z=-980}
    simSpaceSettings {maxAngularAcceleration=1000, maxAngularVelocity=3, maxLinearAcceleration=250, maxLinearVelocity=500, velocityScaleZ=0, worldAlpha=1}
   }
   n6 = RigidBody(ComponentPose=@n5) {
    bClampLinearTranslationLimitToRefPose=true
    bOverrideWorldGravity=true
    overridePhysicsAsset=&/F_MED_Jade_PhysicsArm
    simulationSpace=BaseBoneSpace
    alphaBoolBlend {blendOption=Linear}
    baseBoneRef {boneName=upperarm_r}
    componentAppliedLinearAccClamp=<3000 3000 3000>
    componentLinearAccScale=<0.4 0.3 0.3>
    componentLinearVelScale=<0.4 0.3 0.3>
    overrideWorldGravity {z=-980}
    simSpaceSettings {maxAngularAcceleration=1000, maxAngularVelocity=3, maxLinearAcceleration=250, maxLinearVelocity=500, velocityScaleZ=0, worldAlpha=1}
   }
   n7 = "Constraint - dyn_strap_00_MSTR (PP)"!(ComponentPose=@n6, Alpha=1, ConstraintWeights_0=0.581, ConstraintWeights_1:float=0.419) -> (Pose) {
    alphaBoolBlend {blendOption=Linear}
    boneToModify {boneName=dyn_strap_00_MSTR}
    constraintSetup [{targetBone {boneName=spine_05}}, {targetBone {boneName=clavicle_l}}]
    constraintWeights [1,1]
   }
   n8 = Anim Dynamics - Bone: dyn_zipper(ComponentPose=@n7, Alpha=1) {
    bGravityOverrideInSimSpace=true
    bOverrideLinearDamping=true
    bUseGravityOverride=true
    linearDampingOverride=0.8
    alphaBoolBlend {blendOption=Linear}
    boundBone {boneName=dyn_zipper}
    gravityOverride {z=-980}
    physicsBodyDefinitions [{
      sphereCollisionRadius=0
      boundBone {boneName=dyn_zipper}
      boxExtents=<4 4 4>
      constraintSetup {angularLimitsMax=<0 0 0>, angularLimitsMin=<0 0 0>, angularTarget=<0 0 0>}
      localJointOffset=<0 0 0>
     }]
   }
   n9 = Anim Dynamics - Bone: dyn_strap_01(ComponentPose=@n8, Alpha=1, ExternalForce="0,0,0") {
    bGravityOverrideInSimSpace=true
    bOverrideLinearDamping=true
    bUseGravityOverride=true
    linearDampingOverride=1
    alphaBoolBlend {blendOption=Linear}
    boundBone {boneName=dyn_strap_01}
    gravityOverride {z=-980}
    physicsBodyDefinitions [{
      sphereCollisionRadius=0
      boundBone {boneName=dyn_strap_01}
      boxExtents=<4 4 4>
      constraintSetup {angularLimitsMax=<0 0 0>, angularLimitsMin=<0 0 0>, angularTarget=<0 0 0>}
      localJointOffset=<0 0 0>
     }]
   }
   n10 = Anim Dynamics - Bone: dyn_armzipper(ComponentPose=@n9, Alpha=1, ExternalForce="0,0,0") {
    bGravityOverrideInSimSpace=true
    bOverrideLinearDamping=true
    bUseGravityOverride=true
    linearDampingOverride=1
    alphaBoolBlend {blendOption=Linear}
    boundBone {boneName=dyn_armzipper}
    gravityOverride {z=-980}
    physicsBodyDefinitions [{
      sphereCollisionRadius=0
      boundBone {boneName=dyn_armzipper}
      boxExtents=<4 4 4>
      constraintSetup {angularLimitsMax=<0 0 0>, angularLimitsMin=<0 0 0>, angularTarget=<0 0 0>}
      localJointOffset=<0 0 0>
     }]
   }
   Output Pose(Result=Component To Local(ComponentPose=@n10)) {layerGroup=None}
  ]}
 EventGraph {nodes [Event BlueprintUpdateAnimation()]}
]
variables []
