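# Physics asset for the FaceAcc part of the Jade female medium player body, simulating her bangs and left and right pigtails.
# Capsule bodies cover the hair bones plus head, neck and clavicles, with the unconstrained clavicle capsules apparently serving as colliders.
# Pigtail chains and branching bang chains use tight swing and twist limits, hanging from back-of-head and bang-base bodies that join the head with free angular constraints.
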PhysicsAsset /BRCosmetics/Characters/Player/Female/Medium/Bodies/F_MED_Jade/Meshes/Parts/F_MED_Jade_FaceAcc_PhysicsAsset
bodies {
 clavicle_l [
  {shapeType=Capsule, length=14.25, radius=7, shapeName=clavicle_l_capsule, center=<4.7389 1.7461 1.8026>, rotation=<rot -80.0001 0.0001 0>}
  {shapeType=Capsule, length=22.75, radius=15, shapeName=clavicle_l_capsule, center=<0.0001 16.3595 0>}
 ]
 clavicle_r [{shapeType=Capsule, length=14, radius=4.5, shapeName=clavicle_r_capsule, center=<-3.923 -1.2277 -1.3936>, rotation=<rot 90 0 0>}]
 dyn_backofhead [{shapeType=Capsule, length=1, radius=1, shapeName=dyn_backofhead_capsule}]
 dyn_bang_L_01 [{shapeType=Capsule, length=1, radius=1.5, shapeName=dyn_bang_L_01_capsule, center=<1.2523 0.7652 0.0003>, rotation=<rot 90 0 -10>}]
 dyn_bang_L_02 [{
   shapeType=Capsule
   length=1
   radius=1.5
   shapeName=dyn_bang_L_02_capsule
   center=<1.4575 0.6 0.25>
   rotation=<rot 79.9998 -179.9999 -169.9998>
  }]
 dyn_bang_L_03 [{shapeType=Capsule, length=1, radius=1, shapeName=dyn_bang_L_03_capsule, center=<1 0.6 0.8>, rotation=<rot -16.0011 31.0733 114.5801>}]
 dyn_bang_L_04 [{shapeType=Capsule, length=1, radius=1, shapeName=dyn_bang_L_04_capsule}]
 dyn_bang_L_05 [{shapeType=Capsule, length=1, radius=1, shapeName=dyn_bang_L_05_capsule}]
 dyn_bang_R_01 [{
   shapeType=Capsule
   length=1.1218
   radius=1
   shapeName=dyn_bang_R_01_capsule
   center=<0.5954 0.5525 -0.3442>
   rotation=<rot 2.8862 -72.3588 -97.9626>
  }]
 dyn_bang_R_02 [{shapeType=Capsule, length=1, radius=1, shapeName=dyn_bang_R_02_capsule, center=<1.6789 0.7972 0.1317>, rotation=<rot 90 0 0>}]
 dyn_bang_R_03 [{shapeType=Capsule, length=1, radius=1, shapeName=dyn_bang_R_03_capsule, center=<1 0 -1.2>, rotation=<rot 90 0 0>}]
 dyn_bang_R_04 [{shapeType=Capsule, length=1, radius=1, shapeName=dyn_bang_R_04_capsule, center=<1 0 0>, rotation=<rot 90 0 0>}]
 dyn_bang_R_05 [{shapeType=Capsule, length=1, radius=1, shapeName=dyn_bang_R_05_capsule, center=<1 0 0>, rotation=<rot 90 0 0>}]
 dyn_bangbase [{shapeType=Capsule, length=1, radius=1, shapeName=dyn_bangbase_capsule}]
 dyn_pgtail_L_01 [{shapeType=Capsule, length=2, radius=2, shapeName=dyn_pgtail_L_01_capsule}]
 dyn_pgtail_L_02 [{
   shapeType=Capsule
   length=2
   radius=2
   shapeName=dyn_pgtail_L_02_capsule
   center=<1.1509 1.2294 -1.1831>
   rotation=<rot -49.9999 -100.0006 179.9999>
  }]
 dyn_pgtail_L_03 [{
   shapeType=Capsule
   length=3.5
   radius=3.5
   shapeName=dyn_pgtail_L_03_capsule
   center=<1.4997 -0.2644 0.0004>
   rotation=<rot -90 26.565 -36.5651>
  }]
 dyn_pgtail_L_04 [{
   shapeType=Capsule
   length=3.5
   radius=3.5
   shapeName=dyn_pgtail_L_04_capsule
   center=<1.4997 -0.2644 0.0004>
   rotation=<rot -90 -7.125 -2.8749>
  }]
 dyn_pgtail_L_05 [{shapeType=Capsule, length=3.5, radius=4, shapeName=dyn_pgtail_L_05_capsule, center=<1.4996 0 0.2648>, rotation=<rot -79.9999 -0.0004 0>}]
 dyn_pgtail_L_06 [{
   shapeType=Capsule
   length=3.5
   radius=4
   shapeName=dyn_pgtail_L_06_capsule
   center=<1.3177 0.2645 1.1634>
   rotation=<rot -79.9995 0.0001 10.0006>
  }]
 dyn_pgtail_R_01 [{shapeType=Capsule, length=2, radius=2, shapeName=dyn_pgtail_R_01_capsule}]
 dyn_pgtail_R_02 [{
   shapeType=Capsule
   length=2
   radius=2
   shapeName=dyn_pgtail_R_02_capsule
   center=<0.797 0.5623 0.4356>
   rotation=<rot 1.9697 -178.9672 127.7884>
  }]
 dyn_pgtail_R_03 [{shapeType=Capsule, length=3.5, radius=3.5, shapeName=dyn_pgtail_R_03_capsule, center=<1.6491 0 0>, rotation=<rot 90 -1.2659 -1.2659>}]
 dyn_pgtail_R_04 [{shapeType=Capsule, length=3.5, radius=3.5, shapeName=dyn_pgtail_R_04_capsule, center=<1.6491 0 0>, rotation=<rot 90 -0.0003 -0.0002>}]
 dyn_pgtail_R_05 [{shapeType=Capsule, length=3.5, radius=4, shapeName=dyn_pgtail_R_05_capsule, center=<1.6491 0 0>, rotation=<rot 90 -1.2658 -1.2658>}]
 dyn_pgtail_R_06 [{
   shapeType=Capsule
   length=3.25
   radius=4.25
   shapeName=dyn_pgtail_R_06_capsule
   center=<-0.6724 0 -1.1646>
   rotation=<rot 30.0001 179.9999 179.9999>
  }]
 head [
  {shapeType=Capsule, length=3.0245, radius=8.2276, shapeName=head_capsule, center=<8.2113 -0.057 0.0001>, rotation=<rot 0.0007 10.009 -90>}
  {shapeType=Capsule, length=6.25, radius=5.75, shapeName=head_capsule, center=<2.7675 -0.0001 0>, rotation=<rot 0 0 -0.849>}
 ]
 neck_01 [{shapeType=Capsule, length=4, radius=4, shapeName=neck_01_capsule}]
 neck_02 [{shapeType=Capsule, length=4, radius=4, shapeName=neck_02_capsule}]
}
constraints [
 {bone1Name=dyn_bang_L_01, bone2Name=dyn_bangbase, swing1LimitDegrees=5, swing2LimitDegrees=15, twistLimitDegrees=2}
 {bone1Name=dyn_bang_L_02, bone2Name=dyn_bang_L_01, swing1LimitDegrees=5, swing2LimitDegrees=20, twistLimitDegrees=2}
 {bone1Name=dyn_pgtail_L_01, bone2Name=dyn_backofhead, swing1LimitDegrees=5, swing2LimitDegrees=5, twistLimitDegrees=5}
 {bone1Name=dyn_pgtail_L_02, bone2Name=dyn_pgtail_L_01, swing1LimitDegrees=20, swing2LimitDegrees=20, twistLimitDegrees=5}
 {bone1Name=dyn_pgtail_L_03, bone2Name=dyn_pgtail_L_02, swing1LimitDegrees=20, swing2LimitDegrees=20, twistLimitDegrees=5}
 {bone1Name=dyn_pgtail_L_04, bone2Name=dyn_pgtail_L_03, swing1LimitDegrees=20, swing2LimitDegrees=20, twistLimitDegrees=5}
 {bone1Name=dyn_pgtail_L_05, bone2Name=dyn_pgtail_L_04, swing1LimitDegrees=20, swing2LimitDegrees=20, twistLimitDegrees=5}
 {bone1Name=dyn_pgtail_L_06, bone2Name=dyn_pgtail_L_05, swing1LimitDegrees=20, swing2LimitDegrees=20, twistLimitDegrees=5}
 {bone1Name=dyn_pgtail_R_01, bone2Name=dyn_backofhead, swing1LimitDegrees=5, swing2LimitDegrees=5, twistLimitDegrees=5}
 {bone1Name=dyn_pgtail_R_02, bone2Name=dyn_pgtail_R_01, swing1LimitDegrees=20, swing2LimitDegrees=20, twistLimitDegrees=5}
 {bone1Name=dyn_pgtail_R_03, bone2Name=dyn_pgtail_R_02, swing1LimitDegrees=20, swing2LimitDegrees=20, twistLimitDegrees=5}
 {bone1Name=dyn_pgtail_R_04, bone2Name=dyn_pgtail_R_03, swing1LimitDegrees=20, swing2LimitDegrees=20, twistLimitDegrees=5}
 {bone1Name=dyn_pgtail_R_05, bone2Name=dyn_pgtail_R_04, swing1LimitDegrees=20, swing2LimitDegrees=20, twistLimitDegrees=5}
 {bone1Name=dyn_pgtail_R_06, bone2Name=dyn_pgtail_R_05, swing1LimitDegrees=20, swing2LimitDegrees=20, twistLimitDegrees=5}
 {
  bone1Name=dyn_bangbase
  bone2Name=head
  swing1LimitDegrees=45
  swing1Motion=Free
  swing2LimitDegrees=45
  swing2Motion=Free
  twistLimitDegrees=45
  twistMotion=Free
 }
 {
  bone1Name=dyn_backofhead
  bone2Name=head
  swing1LimitDegrees=45
  swing1Motion=Free
  swing2LimitDegrees=45
  swing2Motion=Free
  twistLimitDegrees=45
  twistMotion=Free
 }
 {
  bone1Name=head
  bone2Name=neck_02
  swing1LimitDegrees=45
  swing1Motion=Free
  swing2LimitDegrees=45
  swing2Motion=Free
  twistLimitDegrees=45
  twistMotion=Free
 }
 {
  bone1Name=neck_02
  bone2Name=neck_01
  swing1LimitDegrees=45
  swing1Motion=Free
  swing2LimitDegrees=45
  swing2Motion=Free
  twistLimitDegrees=45
  twistMotion=Free
 }
 {bone1Name=dyn_bang_R_01, bone2Name=dyn_bangbase, swing1LimitDegrees=5, swing2LimitDegrees=15, twistLimitDegrees=2}
 {bone1Name=dyn_bang_R_02, bone2Name=dyn_bang_R_01, swing1LimitDegrees=5, swing2LimitDegrees=10, twistLimitDegrees=2}
 {bone1Name=dyn_bang_R_03, bone2Name=dyn_bangbase, swing1LimitDegrees=10, swing2LimitDegrees=10, twistLimitDegrees=0}
 {bone1Name=dyn_bang_R_04, bone2Name=dyn_bang_R_03, swing1LimitDegrees=10, swing2LimitDegrees=10, twistLimitDegrees=0}
 {bone1Name=dyn_bang_R_05, bone2Name=dyn_bang_R_03, swing1LimitDegrees=10, swing2LimitDegrees=10, twistLimitDegrees=0}
 {bone1Name=dyn_bang_L_03, bone2Name=dyn_bangbase, swing1LimitDegrees=15, swing2LimitDegrees=10, twistLimitDegrees=0}
 {bone1Name=dyn_bang_L_04, bone2Name=dyn_bang_L_03, swing1LimitDegrees=15, swing2LimitDegrees=10, twistLimitDegrees=0}
 {bone1Name=dyn_bang_L_05, bone2Name=dyn_bang_L_03, swing1LimitDegrees=15, swing2LimitDegrees=10, twistLimitDegrees=0}
]
